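# Sound cue for the TimeInterval back bling cosmetic, using the reactive back bling audio property sheet.
# It picks one of three equally weighted waves at random and feeds it through a modulator that leaves pitch unchanged.
# Playback runs at reduced volume in the gadgets sound class with basic foley attenuation.
# A concurrency override allows one instance per owner, stopping the oldest when a new one plays.
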SoundCue /BRCosmetics/Sounds/Backblings/TimeInterval/Backpack_TimeInterval_Cue
root=Modulator_0
nodes [
 {id=Modulator_0, children [Random_0], properties {pitchMax=1, pitchMin=1, volumeMax=1}}
 {id=Random_0, children [WavePlayer_3,WavePlayer_4,WavePlayer_5], properties {weights [1,1,1]}}
 {id=WavePlayer_3, properties {soundWaveAssetPtr=&/IP_Backpack_TimeInterval_01}}
 {id=WavePlayer_4, properties {soundWaveAssetPtr=&/IP_Backpack_TimeInterval_02}}
 {id=WavePlayer_5, properties {soundWaveAssetPtr=&/IP_Backpack_TimeInterval_03}}
]
properties {
 attenuationSettings=&Foley_Attenuation_Basic_Att
 audioPropertiesSheet=/BRCosmetics/Sounds/Settings/PropertySheets/Reactive/APS_Cosmetic_Reactive_Backbling
 bOverrideConcurrency=true
 maxDistance=1536
 soundClassObject=&Fort_Gadgets
 volumeMultiplier=0.5
 concurrencyOverrides {bLimitToOwner=true, maxCount=1, resolutionRule=StopOldest}
}
